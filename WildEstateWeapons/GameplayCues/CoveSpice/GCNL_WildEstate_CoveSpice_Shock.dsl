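; Looping gameplay cue that shows the CoveSpice shock effect on a player pawn.
; On start it applies a material override, body overlay and custom depth stencil, then picks a weighted random shock montage, played through OnRep_ShockAnim.
; A timer reads charge percent from the target's AC_CoveSpice_ChargePercent component into the looping FX and both dynamic materials.
; Removal clears the override, overlay and timer, but leaves custom depth set.
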(event EventOnLoopingStart(EitherVFX-preferthisversion)
  (params (MyTarget :type Actor%20Object%20Reference) (Parameters :type Gameplay%20Cue%20Parameters%20Structure%20%28by%20ref%29) (ParticleComponents :type Array%20of%20FXSystem%20Component%20Object%20References) (AudioComponents :type Array%20of%20Audio%20Component%20Object%20References))
  (bind _target_player (Variables|Default|GetTarget%20Player))
  (bind _returnvalue (Customization|GetSkeletalMeshforPartType _target_player "Body" :!class /Script/FortniteGame.FortPlayerPawn))
  (bind _charge_percent (Variables|ChargeDamage|GetCharge%20Percent))
  (CallFunction|UpdatePercentCharge MyTarget)
  (Variables|Default|SetLoopingFX (Utilities|Array|Get(acopy) ParticleComponents))
  (bind _returnvalue_2 (Rendering|Material|CreateDynamicMaterialInstance MyTarget (Variables|Default|GetMaterial) "None" "None" :!class /Script/Engine.KismetMaterialLibrary))
  (Variables|Default|SetMID _returnvalue_2)
  (bind _returnvalue_3 (Rendering|Material|CreateDynamicMaterialInstance MyTarget (Variables|Default|GetOverlayMaterial) "None" "None" :!class /Script/Engine.KismetMaterialLibrary))
  (Variables|Default|SetMID_Overlay _returnvalue_3)
  (bind _asfort_player_pawn (Utilities|Casting|CastToFortPlayerPawn MyTarget))
  (Variables|Default|SetTarget%20Player _asfort_player_pawn)
  (Utilities|FlowControl|Sequence
    (:then_0
      (bind (_boutiscurrentoverride _returnvalue_4) (Class|FortPlayerPawn|IsMaterialOverrideApplied (Variables|Default|GetTarget%20Player) (Variables|Default|GetMaterial%20Override%20Handdle)))
      (if (not _returnvalue_4)
        (bind _returnvalue_5 (Class|FortPlayerPawn|ApplyMaterialOverrideLocalwithFlags :self (Variables|Default|GetTarget%20Player) :Material (Utilities|ResolveSoftReference (Variables|Default|GetMID)) :Priority 111.0 :Flags (select (GameplayTags|HasMatchingGameplayTag (Variables|Default|GetTarget%20Player) "(TagName=\"Gameplay.Camera.FirstPerson\")") 21 21)))
        (Variables|Default|SetMaterial%20Override%20Handdle _returnvalue_5)
        (Rendering|Material|SetOverlayMaterial _returnvalue (Variables|Default|GetMID_Overlay) false 0 :!class /Script/Engine.MeshComponent)
        (Rendering|SetRenderCustomDepth _returnvalue true)
        (Rendering|SetCustomDepthStencilValue _returnvalue 5)
        (Effects|Components|ParticleSystem|SetFloatParameter (Variables|Default|GetLoopingFX) "ChargePercent" _charge_percent)
        (Rendering|Material|SetScalarParameterValue (Variables|Default|GetMID) "ChargePercent" _charge_percent :!class /Script/Engine.MaterialInstanceDynamic)
        (Rendering|Material|SetScalarParameterValue (Variables|Default|GetMID_Overlay) "ColorBlendAlpha" _charge_percent :!class /Script/Engine.MaterialInstanceDynamic)
        (bind _returnvalue_6 (Utilities|Time|SetTimerbyEvent (EventDispatchers|CreateEvent :!event Custom|Update%20Charge%20Percent) (ScalableFloat|GetValueatLevel (Variables|ChargeDamage|GetCharge%20Update%20Rate)) true true))
        (Variables|ChargeDamage|SetCharge%20Update%20Timer _returnvalue_6)))
    (:then_1
      (Variables|Default|SetShock%20Anim (CallFunction|GetRandomShockAnim)))))

(event EventOnRemoval(EitherVFX-preferthisversion)
  (params (MyTarget :type Actor%20Object%20Reference) (Parameters :type Gameplay%20Cue%20Parameters%20Structure%20%28by%20ref%29) (ParticleComponents :type Array%20of%20FXSystem%20Component%20Object%20References) (AudioComponents :type Array%20of%20Audio%20Component%20Object%20References) (BurstCameraShakeInstance :type Camera%20Shake%20Base%20Object%20Reference) (BurstDecalInstance :type Decal%20Actor%20Object%20Reference))
  (bind _asfort_player_pawn (Utilities|Casting|CastToFortPlayerPawn MyTarget)
    (:then
      (Rendering|Material|SetOverlayMaterial (Customization|GetSkeletalMeshforPartType _asfort_player_pawn "Body" :!class /Script/FortniteGame.FortPlayerPawn) 0 false 0 :!class /Script/Engine.MeshComponent)
      (Class|FortPlayerPawn|RemoveMaterialOverrideLocal _asfort_player_pawn (Variables|Default|GetMaterial%20Override%20Handdle))
      (Utilities|Time|ClearandInvalidateTimerbyHandle self (Variables|ChargeDamage|GetCharge%20Update%20Timer)))
    (:CastFailed
      (Fortnite|FortShippingLog MyTarget "CoveSpice shock FX material not properly removed from Target Player Pawn!"))))

(event Custom|Update%20Charge%20Percent
  (bind _charge_percent (Variables|ChargeDamage|GetCharge%20Percent))
  (bind _target_player (Variables|Default|GetTarget%20Player)
    (:then
      (CallFunction|UpdatePercentCharge _target_player)
      ;[0 0.1453 0.15 0.5] update material parameters that use the charge percent
      (Utilities|FlowControl|Sequence
        (:then_0
          (bind _loopingfx (Variables|Default|GetLoopingFX)
            (:then
              (Effects|Components|ParticleSystem|SetFloatParameter _loopingfx "ChargePercent" _charge_percent))
            (:else)))
        (:then_1
          (bind _mid (Variables|Default|GetMID)
            (:then
              (Rendering|Material|SetScalarParameterValue _mid "ChargePercent" _charge_percent :!class /Script/Engine.MaterialInstanceDynamic))
            (:else)))
        (:then_2
          (bind _mid_overlay (Variables|Default|GetMID_Overlay)
            (:then
              (Rendering|Material|SetScalarParameterValue _mid_overlay "ColorBlendAlpha" _charge_percent :!class /Script/Engine.MaterialInstanceDynamic))
            (:else)))))
    (:else)))

(fn GetRandomShockAnim
  (bind _current_available_anims (LocalVariables|GetCurrent%20Available%20Anims))
  (LocalVariables|SetTotal%20Weight 0.0)
  (LocalVariables|SetChosen%20Anim%20Montage 0)
  (LocalVariables|SetCurrent%20Available%20Anims (Variables|Default|GetShock%20Anims))
  (Utilities|FlowControl|Sequence
    (:then_0
      ;[0.15 0.04966 0 0.5] determine total weight and roll a random weight
      (for _array_element (Variables|Default|GetShock%20Anims)
        (LocalVariables|SetTotal%20Weight (+ (.Weight_5_7291DD8C46938A138B1AD7928AEC9A21 (Utilities|Struct|BreakSCoveSpiceShockAnimation _array_element)) (LocalVariables|GetTotal%20Weight))))
      (LocalVariables|SetRolled%20Weight (Math|Random|RandomFloatinRange 0.0 (LocalVariables|GetTotal%20Weight))))
    (:then_1
      ;[0.01166 1 0 0.5] shuffle the possible animations, and then find the current one that corresponds with the current rolled weight
      (Utilities|Array|Shuffle _current_available_anims)
      (LocalVariables|SetCurrent%20Check%20Weight 0.0)
      (bind _array_element_2 (Utilities|Array|ForEachLoopwithBreak _current_available_anims)
        (:LoopBody
          (LocalVariables|SetCurrent%20Check%20Weight (+ (.Weight_5_7291DD8C46938A138B1AD7928AEC9A21 (Utilities|Struct|BreakSCoveSpiceShockAnimation _array_element_2)) (LocalVariables|GetCurrent%20Check%20Weight)))
          (LocalVariables|SetCurrent%20Check%20Anim (Utilities|Struct|BreakSCoveSpiceShockAnimation _array_element_2))
          (if (>= (LocalVariables|GetCurrent%20Check%20Weight) (LocalVariables|GetRolled%20Weight))
            (LocalVariables|SetChosen%20Anim%20Montage (LocalVariables|GetCurrent%20Check%20Anim))))))
    (:then_2
      ;[0.2835 0 0.5677 0.5] return the shock anim
      (return (LocalVariables|GetChosen%20Anim%20Montage)))))

(fn OnRep_ShockAnim
  (Animation|Montage|MontagePlaywithBlendIn (Components|SkeletalMesh|GetAnimInstance (Variables|/Script/Engine.Character|GetMesh (Variables|Default|GetTarget%20Player))) (Variables|Default|GetShock%20Anim) (struct :CustomCurve 0 :BlendTime 0.0 :BlendOption "HermiteCubic") 1.0 "MontageLength" 0.0 false))

(fn UpdatePercentCharge
  (params (Actor%20Target :type Actor%20Object%20Reference))
  (bind _returnvalue (Actor|GetComponentbyClass Actor%20Target "/WildEstateWeapons/Gameplay/CoveSpice/AC_CoveSpice_ChargePercent.AC_CoveSpice_ChargePercent_C"))
  (Variables|Default|GetCharge%20Percent%20Component
    (:then
      _)
    (:else
      (Utilities|IsValid _returnvalue
        (:Is%20Valid
          _)
        (:Is%20Not%20Valid
          (Variables|ChargeDamage|SetCharge%20Percent 0.0)))))
  (bind _output_get (Variables|Default|SetCharge%20Percent%20Component _returnvalue))
  (Variables|ChargeDamage|SetCharge%20Percent (Variables|/WildEstateWeapons/Gameplay/CoveSpice/AC_CoveSpice_ChargePercent.AC_CoveSpice_ChargePercent_C|GetCove%20Spice%20Charge%20Percent _output_get)))

(fn ConstructionScript)
