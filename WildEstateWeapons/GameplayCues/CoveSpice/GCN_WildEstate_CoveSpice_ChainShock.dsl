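; Burst gameplay cue for the CoveSpice weapon's Chain Shock effect, drawing its particle effect between the effect causer and the affected target.
; After the parent burst runs, it casts both actors to FortPawn and reads their capsule world locations as start and end points.
; It passes those points as vector parameters to the first spawned particle component, using configurable parameter names.
; A "Chain Shock" debug print remains at the end.
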(fn OnBurst(EitherVFX-preferthisversion)
  (params (MyTarget :type Actor%20Object%20Reference) (Parameters :type Gameplay%20Cue%20Parameters%20Structure%20%28by%20ref%29) (ParticleComponents :type Array%20of%20FXSystem%20Component%20Object%20References) (AudioComponents :type Array%20of%20Audio%20Component%20Object%20References) (BurstCameraShakeInstance :type Camera%20Shake%20Base%20Object%20Reference) (BurstDecalInstance :type Decal%20Actor%20Object%20Reference))
  (bind (_normalizedmagnitude _rawmagnitude _effectcontext _matchedtagname _originaltag _aggregatedsourcetags _aggregatedtargettags _location _normal _instigator _effectcauser) (Ability|GameplayCue|BreakGameplayCueParameters Parameters))
  (bind _output (Utilities|Array|Get(acopy) ParticleComponents))
  (|CallParentFunction MyTarget Parameters ParticleComponents AudioComponents BurstCameraShakeInstance BurstDecalInstance)
  (bind _asfort_pawn (Utilities|Casting|CastToFortPawn MyTarget))
  (LocalVariables|SetEnd%20Location (Transformation|GetWorldLocation (Variables|/Script/Engine.Character|GetCapsuleComponent _asfort_pawn)))
  (bind _asfort_pawn_2 (Utilities|Casting|CastToFortPawn _effectcauser))
  (LocalVariables|SetStart%20Location (Transformation|GetWorldLocation (Variables|/Script/Engine.Character|GetCapsuleComponent _asfort_pawn_2)))
  (Effects|Components|ParticleSystem|SetVectorParameter _output (Variables|Default|GetStart%20Location%20Param) (LocalVariables|GetStart%20Location))
  (Effects|Components|ParticleSystem|SetVectorParameter _output (Variables|Default|GetEnd%20Location%20Param) (LocalVariables|GetEnd%20Location))
  (Development|PrintString _output "Chain Shock")
  (return))
